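# PapayaOptionsMenu is the post-game options menu widget for the Papaya LTM, derived from the native PapayaPostGameOptions class.
# An overlay anchors a minimum-width size box at the bottom right, holding the inherited FortPostGameMenuListView.
# The list reuses the default XP screen's menu entry and item shop entry widgets.
# The event graph holds only unconnected stub events, so the Blueprint supplies layout while behavior is left to the parent class.
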WidgetBlueprint /Game/Athena/HUD/LTM/Papaya/Postgame/PapayaOptionsMenu
graphs [EventGraph {nodes [
   Event PreConstruct()
   Event Construct()
   Event Tick()
  ]}]
variables []
widgets {
 info {inheritedWidgetCount=1, parentClass=/Script/FortniteUI.PapayaPostGameOptions, rootWidgetClass=&Overlay, widgetCount=3}
 widgets [
  Overlay_0 {widgetClassPath=&Overlay, widget=Overlay_0}
  SizeBoxNavigationOptions {
   widgetClassPath=&SizeBox
   bIsVariable=true
   parent=Overlay_0
   widget=SizeBoxNavigationOptions
   properties {bOverride_MinDesiredWidth=true, minDesiredWidth=480}
   slot {class=OverlaySlot, properties {horizontalAlignment=HAlign_Right, verticalAlignment=VAlign_Bottom, padding=<ltrb 0 0 40 80>}}
  }
  ListView_PostGameMenu {
   widgetClassPath=/Script/FortniteUI.FortPostGameMenuListView
   bInherited=true
   bIsVariable=true
   parent=SizeBoxNavigationOptions
   widget=ListView_PostGameMenu
   properties {
    clipping=Inherit
    entryClass=/Game/Athena/HUD/EndGame/PostGameFlowScreens/DefaultScreens/XPScreen/Menu/FortPostGameMenuEntry.FortPostGameMenuEntry_C
    entrySpacing=0
    entryWidgetClass=&FortPostGameMenuEntry_C
    itemShopEntryClass=/Game/Athena/HUD/EndGame/PostGameFlowScreens/DefaultScreens/XPScreen/Menu/FortPostGameMenuItemShopEntry.FortPostGameMenuItemShopEntry_C
    navigation=&self:WidgetTree.ListView_PostGameMenu.WidgetNavigation_0
    numDesignerPreviewEntries=2
   }
   slot {class=SizeBoxSlot}
  }
 ]
}
